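# Physics asset for the Scathach soldier backpack cosmetic mesh.
# It holds a single capsule body on the ROOT bone, offset and rotated to approximate the backpack's volume.
# It has no constraints, so the backpack is represented as one body with no jointed secondary physics.
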PhysicsAsset /BRCosmetics/Accessories/FORT_Backpacks/Mesh/F_MED_Soldier_Backpack_Scathach_PhysicsAsset
bodies {ROOT [{shapeType=Capsule, length=18.7188, radius=18.3274, center=<-4.1428 -12.1389 4.2646>, rotation=<rot 11.2331 -90 26.2543>}]}
constraints []
