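# Physics asset for the Female_Commando_04 backpack, giving capsule bodies to the pack root and its dangling zipper, loops, axe and rope bones.
# Limited-angle constraints hang the zipper, top loop, rope loop, rope end and first front and back axe bones off backpack_root.
# Front and back axe chains each run from the root and from the rope end, meeting at their third bone through looser joints to form closed loops.
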PhysicsAsset /Game/Accessories/FORT_Backpacks/Mesh/Female_Commando_04_Physics
bodies {
 backpack_root [{shapeType=Capsule, length=18.8594, radius=11.1119, center=<-4.7395 -1.9559 -0.3917>, rotation=<rot -1.1245 -83.2832 -89.1695>}]
 dyn_axeBack1 [{shapeType=Capsule, length=19.383, radius=1.8921, center=<5.6973 -0.0699 -0.4351>, rotation=<rot 0 -87.8388 -92.6974>}]
 dyn_axeBack2 [{shapeType=Capsule, length=6.7971, radius=0.5, center=<2.7845 -0.134 -0.0427>, rotation=<rot 56.4862 90.8295 -88.3729>}]
 dyn_axeBack3 [{shapeType=Capsule, length=1.4815, radius=1.0765, center=<0.0842 -0.0028 0.0264>, rotation=<rot 25.6802 -83.5282 -174.5387>}]
 dyn_axeBack4 [{shapeType=Capsule, length=2.2467, radius=1.1092, center=<-1.0679 -0.2225 -0.8455>, rotation=<rot -42.8702 70.3567 -37.6503>}]
 dyn_axeBack5 [{shapeType=Capsule, length=5.3711, radius=0.8613, center=<-2.8289 -0.2005 0.411>, rotation=<rot -0.9702 94.1954 -99.1377>}]
 dyn_axeFront1 [{shapeType=Capsule, length=19.7961, radius=1.669, center=<4.1716 0.0022 -0.2272>, rotation=<rot 0 -87.4359 -91.4776>}]
 dyn_axeFront2 [{shapeType=Capsule, length=6.2619, radius=0.8499, center=<2.657 -0.2844 0.0216>, rotation=<rot 0 83.5781 -89.2122>}]
 dyn_axeFront3 [{shapeType=Capsule, length=1.3795, radius=1.0663, center=<0.0948 0.0363 -0.201>, rotation=<rot -13.5966 -5.1916 -144.8716>}]
 dyn_axeFront4 [{shapeType=Capsule, length=4.6044, radius=0.7473, center=<-1.9533 -0.9735 -1.1467>, rotation=<rot 36.6072 134.8819 -56.843>}]
 dyn_axeFront5 [{shapeType=Capsule, length=5.3849, radius=0.5281, center=<-2.4644 -0.1715 0.2045>, rotation=<rot -15.8806 -84.029 -87.18>}]
 dyn_axeRopeEnd [{shapeType=Capsule, length=4.4036, radius=1.6914, center=<3.1958 0.1605 -0.0167>, rotation=<rot -5.7992 90.3185 -92.1812>}]
 dyn_ropeLoop [{shapeType=Capsule, length=1.9692, radius=1.3843, center=<-1.3058 0.2528 -0.4186>, rotation=<rot -11.4046 -92.3008 -108.8064>}]
 dyn_topLoop [{shapeType=Capsule, length=4.4197, radius=1.4196, center=<-2.4499 0 -1.0122>, rotation=<rot 89.7161 -179.9996 -179.9996>}]
 dyn_zipper [{shapeType=Capsule, length=3.0446, radius=1.4914, center=<2.5659 -1.0798 -0.0225>, rotation=<rot 0 89.5166 -89.2989>}]
}
constraints [
 {bone1Name=dyn_zipper, bone2Name=backpack_root, swing1LimitDegrees=49.92, swing2LimitDegrees=20, twistLimitDegrees=1, twistMotion=Locked}
 {bone1Name=dyn_topLoop, bone2Name=backpack_root, swing1LimitDegrees=45, swing2LimitDegrees=5, twistLimitDegrees=5}
 {bone1Name=dyn_ropeLoop, bone2Name=backpack_root, swing1LimitDegrees=15, swing2LimitDegrees=15, twistLimitDegrees=5}
 {bone1Name=dyn_axeRopeEnd, bone2Name=backpack_root, swing1LimitDegrees=15, swing2LimitDegrees=15, twistLimitDegrees=15}
 {bone1Name=dyn_axeFront5, bone2Name=dyn_axeRopeEnd, swing1LimitDegrees=15, swing2LimitDegrees=45, twistLimitDegrees=15}
 {bone1Name=dyn_axeFront4, bone2Name=dyn_axeFront5, swing1LimitDegrees=15, swing2LimitDegrees=45, twistLimitDegrees=15}
 {bone1Name=dyn_axeFront3, bone2Name=dyn_axeFront4, swing1LimitDegrees=15, swing2LimitDegrees=45, twistLimitDegrees=15}
 {bone1Name=dyn_axeBack5, bone2Name=dyn_axeRopeEnd, swing1LimitDegrees=15, swing2LimitDegrees=15, twistLimitDegrees=15}
 {bone1Name=dyn_axeBack4, bone2Name=dyn_axeBack5, swing1LimitDegrees=15, swing2LimitDegrees=15, twistLimitDegrees=15}
 {bone1Name=dyn_axeBack3, bone2Name=dyn_axeBack4, swing1LimitDegrees=15, swing2LimitDegrees=15, twistLimitDegrees=15}
 {bone1Name=dyn_axeFront1, bone2Name=backpack_root, swing1LimitDegrees=5, swing2LimitDegrees=10, twistLimitDegrees=5}
 {bone1Name=dyn_axeFront2, bone2Name=dyn_axeFront1, swing1LimitDegrees=15, swing2LimitDegrees=15, twistLimitDegrees=15}
 {bone1Name=dyn_axeBack1, bone2Name=backpack_root, swing1LimitDegrees=5, swing2LimitDegrees=5, twistLimitDegrees=5}
 {bone1Name=dyn_axeBack2, bone2Name=dyn_axeBack1, swing1LimitDegrees=15, swing2LimitDegrees=15, twistLimitDegrees=15}
 {bone1Name=dyn_axeBack3, bone2Name=dyn_axeBack2, swing1LimitDegrees=45, swing2LimitDegrees=45, twistLimitDegrees=45}
 {bone1Name=dyn_axeFront3, bone2Name=dyn_axeFront2, swing1LimitDegrees=45, swing2LimitDegrees=45, twistLimitDegrees=45}
]
