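# Animation Blueprint for the Female Commando 04 backpack, adding physics-driven secondary motion and pose corrections.
# Each update maps the owner mesh's spine yaw delta to a small backpack_root rotation and turns the skydiving flag into a blend alpha.
# The ref pose gets that root rotation, feeds a RigidBody simulation in backpack-root space, then three axe-bone roll offsets apply while skydiving.
# The zipper AnimDynamics node and Input Pose are unconnected.
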AnimBlueprint /BRCosmetics/Accessories/FORT_Backpacks/Mesh/Female_Commando_04_Skeleton_AnimBlueprint
target_skeleton=/Game/Accessories/FORT_Backpacks/Mesh/Female_Commando_04_Skeleton
graphs [
 AnimGraph {nodes [
   Anim Dynamics - Bone: dyn_zipper(Alpha=1) {
    angularSpringConstant=20
    bGravityOverrideInSimSpace=true
    bUseGravityOverride=true
    alphaBoolBlend {blendOption=Linear}
    boundBone {boneName=dyn_zipper}
    componentAppliedLinearAccClamp=<1000 1000 1000>
    componentLinearAccScale=<0.25 0.25 0.25>
    componentLinearVelScale=<0.3 0.3 0.3>
    gravityOverride {z=-980}
    physicsBodyDefinitions [{
      sphereCollisionRadius=0
      boundBone {boneName=dyn_zipper}
      boxExtents=<3 2 3>
      constraintSetup {angularLimitsMax=<3 0 0>, angularLimitsMin=<-3 0 -120>, angularTarget=<0 0 0>}
      localJointOffset=<2 0.2 0>
     }]
    relativeSpaceBone {boneName=backpack_root}
   }
   n1 = "Transform (Modify) Bone - Bone: backpack_root"(
    ComponentPose=Local To Component(LocalPose=Local Space Ref Pose())
    Alpha=1
    Translation="0,0,0"
    Rotation=Get backpack_rotate()
    Scale="1,1,1"
   ) {rotationMode=BMM_Additive, rotationSpace=BCS_BoneSpace, alphaBoolBlend {blendOption=Linear}, boneToModify {boneName=backpack_root}}
   n2 = RigidBody(ComponentPose=@n1) {
    bOverrideWorldGravity=true
    overridePhysicsAsset=/Game/Accessories/FORT_Backpacks/Mesh/Female_Commando_04_Physics
    simulationSpace=BaseBoneSpace
    alphaBoolBlend {blendOption=Linear}
    baseBoneRef {boneName=backpack_root}
    componentAppliedLinearAccClamp=<1000 1000 1000>
    componentLinearAccScale=<0.3 0.3 0.3>
    componentLinearVelScale=<0.35 0.35 0.35>
    overrideWorldGravity {z=-980}
   }
   n3 = Get SkyDive_correction() -> (SkyDive_correction:double)
   n4 = "Transform (Modify) Bone - Bone: dyn_axeRopeEnd"(ComponentPose=@n2, Alpha=@n3, Translation="0,0,0", Rotation="0,0,45", Scale="1,1,1") {rotationMode=BMM_Additive, alphaBoolBlend {blendOption=Linear}, boneToModify {boneName=dyn_axeRopeEnd}, rotation {roll=45}}
   n5 = "Transform (Modify) Bone - Bone: dyn_axeFront2"(ComponentPose=@n4, Alpha=@n3, Translation="0,0,0", Rotation="0,0,45", Scale="1,1,1") {rotationMode=BMM_Additive, alphaBoolBlend {blendOption=Linear}, boneToModify {boneName=dyn_axeFront2}, rotation {roll=45}}
   n6 = "Transform (Modify) Bone - Bone: dyn_axeBack2"(ComponentPose=@n5, Alpha=@n3, Translation="0,0,0", Rotation="0,0,45", Scale="1,1,1") {rotationMode=BMM_Additive, alphaBoolBlend {blendOption=Linear}, boneToModify {boneName=dyn_axeBack2}, rotation {roll=45}}
   Output Pose(Result=Component To Local(ComponentPose=@n6)) {layerGroup=None}
   AnimGraphNode_SubInput_0 = Input Pose()
  ]}
 EventGraph {nodes [
   Event BlueprintUpdateAnimation()
   n1 = GetOwnerSkeletalMeshComponent()
   n2 = BreakRotator(InRot=BreakTransform(InTransform=GetDeltaTransformFromRefPose(self=@n1, BoneName=spine_05, BaseName=spine_01)).Rotation)
   > Set backpack_rotate(backpack_rotate=MakeRotator(Yaw=MapRangeClamped(Value=@n2.Yaw, InRangeA=12, InRangeB=-10, OutRangeA=8, OutRangeB=10)))
   > Set SkyDive_correction(SkyDive_correction:double=SelectFloat(A=1, B=0, bPickA=Get bIsSkyDiving())) -> (Output_Get:double)
   TryGetPawnOwner()
  ]}
]
variables [
 backpack_rotate {type=Rotator}
 SkyDive_correction {type=number}
]
