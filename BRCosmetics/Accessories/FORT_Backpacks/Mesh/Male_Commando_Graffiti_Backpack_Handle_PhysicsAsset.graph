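# Physics asset for the handle of the Male Commando Graffiti backpack cosmetic, likely letting the handle swing as a secondary-motion part.
# Two sphere bodies cover the bones: a large offset one on dyn_handle and a small one on root.
# A single constraint joins dyn_handle to root with both swing axes locked, leaving twist as the only rotational axis so the handle moves like a hinge.
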PhysicsAsset /BRCosmetics/Accessories/FORT_Backpacks/Mesh/Male_Commando_Graffiti_Backpack_Handle_PhysicsAsset
bodies {dyn_handle [{shapeType=Sphere, radius=13.0102, center=<0 0 -12>}], root [{shapeType=Sphere, radius=2.247, center=<-7.6982 -8.0334 0.4148>}]}
constraints [{
  bone1Name=dyn_handle
  bone2Name=root
  swing1LimitDegrees=0
  swing1Motion=Locked
  swing2LimitDegrees=0
  swing2Motion=Locked
  twistLimitDegrees=30
 }]
