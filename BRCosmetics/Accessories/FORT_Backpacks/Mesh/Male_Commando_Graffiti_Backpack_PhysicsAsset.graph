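# Physics asset for the Male Commando graffiti backpack cosmetic, giving its hanging mask and three cans simple bodies for secondary motion.
# Each dyn_ bone for the mask and the left, middle and right cans has a capsule, while the root bone has a small sphere.
# All four capsules constrain directly to root with zero swing1 and narrow swing2 and twist limits, with the mask allowed the most movement.
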PhysicsAsset /BRCosmetics/Accessories/FORT_Backpacks/Mesh/Male_Commando_Graffiti_Backpack_PhysicsAsset
bodies {
 dyn_can_left [{shapeType=Capsule, length=4.057, radius=3.5046, shapeName=dyn_can_left_capsule, center=<0.318 0 -3.0203>, rotation=<rot -0.0001 -90 -4.2017>}]
 dyn_can_mid [{shapeType=Capsule, length=5.6947, radius=3.0906, shapeName=dyn_can_mid_capsule, center=<-0.0104 -0.0708 -2.0684>, rotation=<rot 0 -90 0>}]
 dyn_can_right [{shapeType=Capsule, length=10.8758, radius=3.1463, shapeName=dyn_can_right_capsule, center=<0 0 -3.2>, rotation=<rot 0 -90 0>}]
 dyn_mask [{
   shapeType=Capsule
   length=12.1158
   radius=5.1537
   shapeName=dyn_mask_capsule
   center=<-1.5696 -1.9894 -13.6459>
   rotation=<rot 0 -86.6304 -76.3944>
  }]
 root [{shapeType=Sphere, radius=2.247, center=<-7.6982 -9.0663 0.4148>}]
}
constraints [
 {bone1Name=dyn_mask, bone2Name=root, swing1LimitDegrees=0, swing2LimitDegrees=10, twistLimitDegrees=25}
 {bone1Name=dyn_can_mid, bone2Name=root, swing1LimitDegrees=0, swing2LimitDegrees=6, twistLimitDegrees=5}
 {bone1Name=dyn_can_left, bone2Name=root, swing1LimitDegrees=0, swing2LimitDegrees=5, twistLimitDegrees=5}
 {bone1Name=dyn_can_right, bone2Name=root, swing1LimitDegrees=0, swing2LimitDegrees=5, twistLimitDegrees=5}
]
