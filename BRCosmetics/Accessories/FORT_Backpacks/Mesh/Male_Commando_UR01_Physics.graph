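# Physics asset for the Male_Commando_UR01 backpack accessory, set up for dangling flask and shovel parts on the bag.
# Capsule bodies cover the Bag bone and the dyn_Flask and dyn_Shovel bones, each of which is constrained directly to Bag.
# The flask's limits leave a single swing axis for hinge-like motion, while the shovel has a tight, equal range on both swings and twist.
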PhysicsAsset /BRCosmetics/Accessories/FORT_Backpacks/Mesh/Male_Commando_UR01_Physics
bodies {
 Bag [{shapeType=Capsule, length=17.4845, radius=16.0229}]
 dyn_Flask [{shapeType=Capsule, length=11.4172, radius=8.0226, center=<-3.0236 -0.8868 -3.1272>, rotation=<rot 5.9076 -90 -25.5237>}]
 dyn_Shovel [{shapeType=Capsule, length=27.9852, radius=8.4166, center=<-0.5203 1.206 -6.9796>, rotation=<rot 0 0 -9.9187>}]
}
constraints [
 {bone1Name=dyn_Flask, bone2Name=Bag, swing1LimitDegrees=0, swing2LimitDegrees=45, twistLimitDegrees=0}
 {bone1Name=dyn_Shovel, bone2Name=Bag, swing1LimitDegrees=10, swing2LimitDegrees=10, twistLimitDegrees=10}
]
