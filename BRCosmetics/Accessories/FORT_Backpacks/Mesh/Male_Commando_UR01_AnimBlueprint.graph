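# Post-style animation blueprint for the Male_Commando_UR01 backpack skeleton that adds secondary physics motion to an incoming pose.
# The active path runs Input Pose through a RigidBody node using the Male_Commando_UR01 physics asset, simulated in Root bone space, then to the output.
# An older-looking chain of AnimDynamics nodes for the Bag, flask and shovel bones remains disconnected from the output.
# The event graph holds only empty events.
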AnimBlueprint /BRCosmetics/Accessories/FORT_Backpacks/Mesh/Male_Commando_UR01_AnimBlueprint
target_skeleton=/Game/Accessories/FORT_Backpacks/Mesh/Male_Commando_UR01_Skeleton
graphs [
 AnimGraph {nodes [
   n1 = Anim Dynamics - Bone: Bag(Alpha=0.7, ExternalForce="0,0,0") {
    angularSpringConstant=20
    bAngularSpring=true
    bGravityOverrideInSimSpace=true
    bLinearSpring=true
    linearSpringConstant=100
    simulationSpace=Actor
    alphaBoolBlend {blendOption=Linear}
    boundBone {boneName=Bag}
    physicsBodyDefinitions [{
      sphereCollisionRadius=0
      boundBone {boneName=Bag}
      boxExtents=<20 5 20>
      constraintSetup {angularLimitsMax=<2 0.1 0>, angularLimitsMin=<-2 0.1 0>, angularTarget=<0 0 0>, linearAxesMax=<0 0 5>, linearAxesMin=<0 0 -2>}
      localJointOffset=<0 0 -10>
     }]
   }
   n2 = Anim Dynamics - Bone: dyn_Flask(ComponentPose=@n1, Alpha=0.7, ExternalForce="0,0,0") {
    bGravityOverrideInSimSpace=true
    alphaBoolBlend {blendOption=Linear}
    boundBone {boneName=dyn_Flask}
    physicsBodyDefinitions [{
      sphereCollisionRadius=0
      boundBone {boneName=dyn_Flask}
      boxExtents=<20 20 5>
      constraintSetup {angularLimitsMax=<2 20 0>, angularLimitsMin=<-2 -20 0>, angularTarget=<0 0 0>}
      localJointOffset=<0 0 -5>
     }]
    relativeSpaceBone {boneName=Bag}
   }
   Anim Dynamics - Bone: dyn_Shovel(ComponentPose=@n2, Alpha=0.7, ExternalForce="0,0,0") {
    angularSpringConstant=15
    bAngularSpring=true
    bGravityOverrideInSimSpace=true
    bLinearSpring=true
    linearSpringConstant=20
    alphaBoolBlend {blendOption=Linear}
    boundBone {boneName=dyn_Shovel}
    physicsBodyDefinitions [{
      sphereCollisionRadius=0
      boundBone {boneName=dyn_Shovel}
      boxExtents=<5 5 20>
      constraintSetup {twistAxis=AxisZ, angularLimitsMax=<5 0 2>, angularLimitsMin=<-5 0 -2>, angularTarget=<0 0 0>}
      localJointOffset=<0 0 -2>
     }]
   }
   AnimGraphNode_SubInput_0 = Input Pose()
   n3 = RigidBody(ComponentPose=Local To Component(LocalPose=@AnimGraphNode_SubInput_0)) {
    bClampLinearTranslationLimitToRefPose=true
    overridePhysicsAsset=&/Male_Commando_UR01_Physics
    simulationSpace=BaseBoneSpace
    alphaBoolBlend {blendOption=Linear}
    baseBoneRef {boneName=Root}
    componentLinearAccScale=<1 1 1>
    componentLinearVelScale=<1 1 1>
   }
   Output Pose(Result=Component To Local(ComponentPose=@n3)) {layerGroup=None}
  ]}
 EventGraph {nodes [
   Event BlueprintUpdateAnimation()
   Event BlueprintInitializeAnimation()
  ]}
]
variables []
