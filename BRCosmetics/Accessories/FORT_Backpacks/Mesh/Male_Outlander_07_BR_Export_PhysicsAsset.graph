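# Physics asset for the Male Outlander 07 backpack skeletal mesh, a Battle Royale cosmetic back accessory.
# It holds a single body on the Backpack_MESH bone, approximated by one capsule offset from the bone origin and rotated to fit the pack.
# It has no constraints, so there are no jointed or articulated parts.
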PhysicsAsset /BRCosmetics/Accessories/FORT_Backpacks/Mesh/Male_Outlander_07_BR_Export_PhysicsAsset
bodies {Backpack_MESH [{shapeType=Capsule, length=15.3311, radius=10.7447, center=<0 -5.979 -2.7453>, rotation=<rot 0.8759 -90 0>}]}
constraints []
