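# Anim Blueprint for the RavenQuill backpack that adds secondary physics motion to its ring and three feather-tip bones.
# The live chain offsets backpack_root while skydive diving or floating, then runs Anim Dynamics on the ring and spring-driven, root-relative feather tips.
# The update event lowers feather DynamicsAlpha while sliding, with interpolation on the feather nodes.
# Two other Anim Dynamics chains remain in the graph, unconnected to the output.
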AnimBlueprint /BRCosmetics/Accessories/FORT_Backpacks/Mesh/SK_Backpack_RavenQuill_AnimBlueprint
target_skeleton=&/SK_Backpack_RavenQuill_Skeleton
graphs [
 AnimGraph {nodes [
   n1 = Anim Dynamics - Bone: ring(Alpha=1, ExternalForce="0,0,0") {
    bGravityOverrideInSimSpace=true
    lODThreshold=1
    numSolverIterationsPostUpdate=4
    numSolverIterationsPreUpdate=16
    simulationSpace=Actor
    alphaBoolBlend {blendOption=Linear}
    boundBone {boneName=ring}
    physicsBodyDefinitions [{
      sphereCollisionRadius=0
      boundBone {boneName=ring}
      boxExtents=<5 5 5>
      constraintSetup {angularTargetAxis=AxisZ, twistAxis=AxisZ, angularLimitsMax=<5 30 30>, angularLimitsMin=<-90 -30 -30>, angularTarget=<0 0 0>}
      localJointOffset=<0 0 -0.5>
     }]
   }
   n2 = Anim Dynamics - Bone: feathertip_01(ComponentPose=@n1, Alpha=1, ExternalForce="0,0,0") {
    bGravityOverrideInSimSpace=true
    gravityScale=0.15
    lODThreshold=1
    numSolverIterationsPostUpdate=4
    numSolverIterationsPreUpdate=16
    simulationSpace=Actor
    alphaBoolBlend {blendOption=Linear}
    boundBone {boneName=feathertip_01}
    physicsBodyDefinitions [{
      sphereCollisionRadius=0
      boundBone {boneName=feathertip_01}
      boxExtents=<20 20 1>
      constraintSetup {
       angularTargetAxis=AxisY
       twistAxis=AxisY
       angularLimitsMax=<0 30 5>
       angularLimitsMin=<0 -30 -5>
       angularTarget=<0 0 0>
       linearAxesMax=<0 0 5>
       linearAxesMin=<0 0 -1>
      }
      localJointOffset=<0 -1 0>
     }]
   }
   n3 = Anim Dynamics - Bone: feathertip_02(ComponentPose=@n2, Alpha=1, ExternalForce="0,0,0") {
    bGravityOverrideInSimSpace=true
    gravityScale=0.15
    lODThreshold=1
    numSolverIterationsPostUpdate=4
    numSolverIterationsPreUpdate=16
    simulationSpace=Actor
    alphaBoolBlend {blendOption=Linear}
    boundBone {boneName=feathertip_02}
    physicsBodyDefinitions [{
      sphereCollisionRadius=0
      boundBone {boneName=feathertip_02}
      boxExtents=<20 20 1>
      constraintSetup {
       angularTargetAxis=AxisY
       twistAxis=AxisY
       angularLimitsMax=<0 30 5>
       angularLimitsMin=<0 -30 -5>
       angularTarget=<0 0 0>
       linearAxesMax=<0 0 5>
       linearAxesMin=<0 0 -1>
      }
      localJointOffset=<0 0 -1>
     }]
   }
   Anim Dynamics - Bone: feathertip_03(ComponentPose=@n3, Alpha=1, ExternalForce="0,0,0") {
    bGravityOverrideInSimSpace=true
    gravityScale=0.15
    lODThreshold=1
    numSolverIterationsPostUpdate=4
    numSolverIterationsPreUpdate=16
    simulationSpace=Actor
    alphaBoolBlend {blendOption=Linear}
    boundBone {boneName=feathertip_03}
    physicsBodyDefinitions [{
      sphereCollisionRadius=0
      boundBone {boneName=feathertip_03}
      boxExtents=<20 20 1>
      constraintSetup {
       angularTargetAxis=AxisY
       twistAxis=AxisY
       angularLimitsMax=<0 30 5>
       angularLimitsMin=<0 -30 -5>
       angularTarget=<0 0 0>
       linearAxesMax=<0 0 5>
       linearAxesMin=<0 0 -1>
      }
      localJointOffset=<-1 0 0>
     }]
   }
   n4 = Anim Dynamics - Bone: ring!(ComponentPose, Alpha=1, GravityOverride:Vector="0,0,-980", ExternalForce="0,0,0") -> (Pose) {
    bGravityOverrideInSimSpace=true
    lODThreshold=1
    numSolverIterationsPostUpdate=4
    numSolverIterationsPreUpdate=16
    alphaBoolBlend {blendOption=Linear}
    boundBone {boneName=ring}
    componentAppliedLinearAccClamp=<5000 5000 5000>
    componentLinearAccScale=<0.3 0.3 0.3>
    componentLinearVelScale=<0.3 0.3 0.3>
    gravityOverride {z=-980}
    physicsBodyDefinitions [{
      sphereCollisionRadius=0
      boundBone {boneName=ring}
      boxExtents=<5 5 5>
      constraintSetup {angularTargetAxis=AxisZ, twistAxis=AxisZ, angularLimitsMax=<5 30 30>, angularLimitsMin=<-90 -30 -30>, angularTarget=<0 0 0>}
      localJointOffset=<0 0 -0.5>
     }]
    relativeSpaceBone {boneName=backpack_root}
   }
   n5 = Get DynamicsAlpha() -> (DynamicsAlpha:double)
   n6 = Anim Dynamics - Bone: feathertip_01!(ComponentPose=@n4, Alpha=@n5, GravityOverride:Vector="0,0,-100", ExternalForce="0,0,0") -> (Pose) {
    angularSpringConstant=100
    bAngularSpring=true
    bGravityOverrideInSimSpace=true
    bLinearSpring=true
    bUseGravityOverride=true
    gravityScale=0.1
    lODThreshold=1
    linearSpringConstant=150
    numSolverIterationsPostUpdate=4
    numSolverIterationsPreUpdate=16
    simulationSpace=World
    alphaBoolBlend {blendOption=Linear}
    boundBone {boneName=feathertip_01}
    componentAppliedLinearAccClamp=<3000 3000 3000>
    componentLinearAccScale=<0.05 0.05 0.05>
    componentLinearVelScale=<0.05 0.05 0.05>
    gravityOverride {z=-980}
    physicsBodyDefinitions [{
      sphereCollisionRadius=0
      boundBone {boneName=feathertip_01}
      boxExtents=<20 20 1>
      constraintSetup {
       angularTargetAxis=AxisY
       twistAxis=AxisY
       angularLimitsMax=<0 30 5>
       angularLimitsMin=<0 -30 -5>
       angularTarget=<0 0 0>
       linearAxesMax=<0 0 5>
       linearAxesMin=<0 0 -1>
      }
      localJointOffset=<0 -1 0>
     }]
    relativeSpaceBone {boneName=backpack_root}
   }
   n7 = Anim Dynamics - Bone: feathertip_02!(ComponentPose=@n6, Alpha=@n5, GravityOverride:Vector="0,0,-100", ExternalForce="0,0,0") -> (Pose) {
    angularSpringConstant=100
    bAngularSpring=true
    bGravityOverrideInSimSpace=true
    bLinearSpring=true
    bUseGravityOverride=true
    gravityScale=0.1
    lODThreshold=1
    linearSpringConstant=150
    numSolverIterationsPostUpdate=4
    numSolverIterationsPreUpdate=16
    simulationSpace=World
    alphaBoolBlend {blendOption=Linear}
    boundBone {boneName=feathertip_02}
    componentAppliedLinearAccClamp=<3000 3000 3000>
    componentLinearAccScale=<0.05 0.05 0.05>
    componentLinearVelScale=<0.05 0.05 0.05>
    gravityOverride {z=-980}
    physicsBodyDefinitions [{
      sphereCollisionRadius=0
      boundBone {boneName=feathertip_02}
      boxExtents=<20 20 1>
      constraintSetup {
       angularTargetAxis=AxisY
       twistAxis=AxisY
       angularLimitsMax=<0 30 5>
       angularLimitsMin=<0 -30 -5>
       angularTarget=<0 0 0>
       linearAxesMax=<0 0 5>
       linearAxesMin=<0 0 -1>
      }
      localJointOffset=<0 0 -1>
     }]
    relativeSpaceBone {boneName=backpack_root}
   }
   Anim Dynamics - Bone: feathertip_03!(ComponentPose=@n7, Alpha=@n5, GravityOverride:Vector="0,0,-100", ExternalForce="0,0,0") -> (Pose) {
    angularSpringConstant=100
    bAngularSpring=true
    bGravityOverrideInSimSpace=true
    bLinearSpring=true
    bUseGravityOverride=true
    gravityScale=0.1
    lODThreshold=1
    linearSpringConstant=150
    numSolverIterationsPostUpdate=4
    numSolverIterationsPreUpdate=16
    simulationSpace=World
    alphaBoolBlend {blendOption=Linear}
    boundBone {boneName=feathertip_03}
    componentAppliedLinearAccClamp=<3000 3000 3000>
    componentLinearAccScale=<0.05 0.05 0.05>
    componentLinearVelScale=<0.05 0.05 0.05>
    gravityOverride {z=-980}
    physicsBodyDefinitions [{
      sphereCollisionRadius=0
      boundBone {boneName=feathertip_03}
      boxExtents=<20 20 1>
      constraintSetup {
       angularTargetAxis=AxisY
       twistAxis=AxisY
       angularLimitsMax=<0 30 5>
       angularLimitsMin=<0 -30 -5>
       angularTarget=<0 0 0>
       linearAxesMax=<0 0 5>
       linearAxesMin=<0 0 -1>
      }
      localJointOffset=<-1 0 0>
     }]
    relativeSpaceBone {boneName=backpack_root}
   }
   AnimGraphNode_SubInput_0 = Input Pose()
   n8 = "Transform (Modify) Bone - Bone: backpack_root"!(
    ComponentPose=Local To Component(LocalPose=@AnimGraphNode_SubInput_0)
    bAlphaBoolEnabled:bool=Get skydive Rot() -> (skydive Rot:bool)
    Translation="0,-1,0"
    Rotation="0.000,10.000,0.000"
    Scale="1,1,1"
   ) -> (Pose) {
    alphaInputType=Bool
    rotationMode=BMM_Additive
    rotationSpace=BCS_BoneSpace
    translationMode=BMM_Additive
    translationSpace=BCS_BoneSpace
    alphaBoolBlend {blendOption=Linear}
    boneToModify {boneName=backpack_root}
    rotation {yaw=10.0002}
    translation {y=-1}
   }
   n9 = Anim Dynamics - Bone: ring(ComponentPose=@n8, Alpha=1, GravityOverride="0,0,-980", ExternalForce="0,0,0") {
    bGravityOverrideInSimSpace=true
    lODThreshold=1
    numSolverIterationsPostUpdate=4
    numSolverIterationsPreUpdate=16
    alphaBoolBlend {blendOption=Linear}
    boundBone {boneName=ring}
    componentAppliedLinearAccClamp=<5000 5000 5000>
    componentLinearAccScale=<0.3 0.3 0.3>
    componentLinearVelScale=<0.3 0.3 0.3>
    gravityOverride {z=-980}
    physicsBodyDefinitions [{
      sphereCollisionRadius=0
      boundBone {boneName=ring}
      boxExtents=<5 5 5>
      constraintSetup {angularTargetAxis=AxisZ, twistAxis=AxisZ, angularLimitsMax=<5 30 30>, angularLimitsMin=<-90 -30 -30>, angularTarget=<0 0 0>}
      localJointOffset=<0 0 -0.5>
     }]
    relativeSpaceBone {boneName=backpack_root}
   }
   n10 = Get DynamicsAlpha()
   n11 = Anim Dynamics - Bone: feathertip_01(ComponentPose=@n9, Alpha=@n10, GravityOverride="0,0,-100", ExternalForce="0,0,0") {
    angularSpringConstant=100
    bAngularSpring=true
    bGravityOverrideInSimSpace=true
    bLinearSpring=true
    bUseGravityOverride=true
    gravityScale=0.1
    lODThreshold=1
    linearSpringConstant=150
    numSolverIterationsPostUpdate=4
    numSolverIterationsPreUpdate=16
    simulationSpace=RootRelative
    alphaBoolBlend {blendOption=Linear}
    alphaScaleBiasClamp {bInterpResult=true, interpSpeedDecreasing=2}
    boundBone {boneName=feathertip_01}
    componentAppliedLinearAccClamp=<3000 3000 3000>
    componentLinearAccScale=<0.3 0.3 0.2>
    componentLinearVelScale=<0.3 0.3 0.2>
    gravityOverride {z=-980}
    physicsBodyDefinitions [{
      sphereCollisionRadius=0
      boundBone {boneName=feathertip_01}
      boxExtents=<20 20 1>
      constraintSetup {
       angularTargetAxis=AxisY
       twistAxis=AxisY
       angularLimitsMax=<0 30 5>
       angularLimitsMin=<0 -30 -5>
       angularTarget=<0 0 0>
       linearAxesMax=<0 0 5>
       linearAxesMin=<0 0 -1>
      }
      localJointOffset=<0 -1 0>
     }]
    relativeSpaceBone {boneName=backpack_root}
   }
   n12 = Anim Dynamics - Bone: feathertip_02(ComponentPose=@n11, Alpha=@n10, GravityOverride="0,0,-100", ExternalForce="0,0,0") {
    angularSpringConstant=100
    bAngularSpring=true
    bGravityOverrideInSimSpace=true
    bLinearSpring=true
    bUseGravityOverride=true
    gravityScale=0.1
    lODThreshold=1
    linearSpringConstant=150
    numSolverIterationsPostUpdate=4
    numSolverIterationsPreUpdate=16
    simulationSpace=RootRelative
    alphaBoolBlend {blendOption=Linear}
    alphaScaleBiasClamp {bInterpResult=true, interpSpeedDecreasing=2}
    boundBone {boneName=feathertip_02}
    componentAppliedLinearAccClamp=<3000 3000 3000>
    componentLinearAccScale=<0.3 0.3 0.2>
    componentLinearVelScale=<0.3 0.3 0.2>
    gravityOverride {z=-980}
    physicsBodyDefinitions [{
      sphereCollisionRadius=0
      boundBone {boneName=feathertip_02}
      boxExtents=<20 20 1>
      constraintSetup {
       angularTargetAxis=AxisY
       twistAxis=AxisY
       angularLimitsMax=<0 30 5>
       angularLimitsMin=<0 -30 -5>
       angularTarget=<0 0 0>
       linearAxesMax=<0 0 5>
       linearAxesMin=<0 0 -1>
      }
      localJointOffset=<0 0 -1>
     }]
    relativeSpaceBone {boneName=backpack_root}
   }
   n13 = Anim Dynamics - Bone: feathertip_03(ComponentPose=@n12, Alpha=@n10, GravityOverride="0,0,-100", ExternalForce="0,0,0") {
    angularSpringConstant=100
    bAngularSpring=true
    bGravityOverrideInSimSpace=true
    bLinearSpring=true
    bUseGravityOverride=true
    gravityScale=0.1
    lODThreshold=1
    linearSpringConstant=150
    numSolverIterationsPostUpdate=4
    numSolverIterationsPreUpdate=16
    simulationSpace=RootRelative
    alphaBoolBlend {blendOption=Linear}
    alphaScaleBiasClamp {bInterpResult=true, interpSpeedDecreasing=2}
    boundBone {boneName=feathertip_03}
    componentAppliedLinearAccClamp=<3000 3000 3000>
    componentLinearAccScale=<0.3 0.3 0.2>
    componentLinearVelScale=<0.3 0.3 0.2>
    gravityOverride {z=-980}
    physicsBodyDefinitions [{
      sphereCollisionRadius=0
      boundBone {boneName=feathertip_03}
      boxExtents=<20 20 1>
      constraintSetup {
       angularTargetAxis=AxisY
       twistAxis=AxisY
       angularLimitsMax=<0 30 5>
       angularLimitsMin=<0 -30 -5>
       angularTarget=<0 0 0>
       linearAxesMax=<0 0 5>
       linearAxesMin=<0 0 -1>
      }
      localJointOffset=<-1 0 0>
     }]
    relativeSpaceBone {boneName=backpack_root}
   }
   Output Pose(Result=Component To Local(ComponentPose=@n13)) {layerGroup=None}
  ]}
 EventGraph {nodes [
   Event BlueprintUpdateAnimation()
   > Set DynamicsAlpha(DynamicsAlpha:double=SelectFloat(A=0.2, B=0.85, bPickA=Get bIsSliding())) -> (Output_Get:double)
   > n1 = Branch(Condition=OR Boolean(A=Get bIsSkydiveDiveMode(), B=Get bIsSkydiveFloating()))
   > Set skydive Rot(skydive Rot:bool=true) -> (Output_Get:bool)
   Set skydive Rot(execute=@n1.else, skydive Rot=false)
   TryGetPawnOwner()
   SelectFloat(A=0.1, B=0.6, bPickA=Get bIsSliding())
  ]}
]
variables [
 DynamicsAlpha {type=number, default=0.4}
 skydive Rot {type=boolean}
]
