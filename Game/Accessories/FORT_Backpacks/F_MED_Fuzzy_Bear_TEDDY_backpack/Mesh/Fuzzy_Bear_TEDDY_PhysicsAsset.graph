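# Physics asset for the Fuzzy Bear TEDDY backpack mesh, giving its bow a jointed rig for secondary motion.
# All bodies are capsules, with two on root for the main volume.
# Constraints chain root to bow_mstr to bow_offset_anim, which branches into left and right three-bone loops and two-bone tails.
# Swing and twist limits are wide at the bow base, moderate on the tails and tightest on the loops.
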PhysicsAsset /Game/Accessories/FORT_Backpacks/F_MED_Fuzzy_Bear_TEDDY_backpack/Mesh/Fuzzy_Bear_TEDDY_PhysicsAsset
bodies {
 bow_mstr [{shapeType=Capsule, length=6.1875, radius=17.3438, center=<0 19.6756 -8.5357>, rotation=<rot 5.6249 -90 0>}]
 bow_offset_anim [{shapeType=Capsule, length=0, radius=3.0252}]
 loop_L [{shapeType=Capsule, length=2, radius=4, shapeName=loop_L_capsule, center=<1 0 0>, rotation=<rot 0 0 90>}]
 loop_L1 [{shapeType=Capsule, length=6.25, radius=4, shapeName=loop_L1_capsule, center=<1 0 0>, rotation=<rot 0 0 90>}]
 loop_L2 [{shapeType=Capsule, length=9, radius=5, shapeName=loop_L2_capsule, center=<1 0 0>, rotation=<rot 0 0 90>}]
 loop_R [{shapeType=Capsule, length=2, radius=4, shapeName=loop_R_capsule, center=<1 0 0>, rotation=<rot 0 0 90>}]
 loop_R1 [{shapeType=Capsule, length=6.25, radius=4, shapeName=loop_R1_capsule, center=<1 0 0>, rotation=<rot 0 0 90>}]
 loop_R2 [{shapeType=Capsule, length=9, radius=5, shapeName=loop_R2_capsule, center=<1 0 0>, rotation=<rot 0 0 90>}]
 root [
  {shapeType=Capsule, length=22.4375, radius=11.0675, center=<5.2874 12.8818 0.1798>, rotation=<rot -5.6251 -89.9999 0>}
  {shapeType=Capsule, length=26.8137, radius=13.4897, shapeName=root_capsule, center=<-18.4965 11.6594 0>}
 ]
 tail_L [{shapeType=Capsule, length=4, radius=4, shapeName=tail_L_capsule, center=<6 0 0>, rotation=<rot 90 0 0>}]
 tail_L1 [{shapeType=Capsule, length=4.25, radius=4.5, shapeName=tail_L1_capsule, center=<6 0 0>, rotation=<rot 90 0 0>}]
 tail_R [{shapeType=Capsule, length=4, radius=4, shapeName=tail_R_capsule, center=<6 0 0>, rotation=<rot 90 0 0>}]
 tail_R1 [{shapeType=Capsule, length=4.25, radius=4.5, shapeName=tail_R1_capsule, center=<6 0 0>, rotation=<rot 90 0 0>}]
}
constraints [
 {bone1Name=loop_L, bone2Name=bow_offset_anim, swing1LimitDegrees=10, swing2LimitDegrees=10, twistLimitDegrees=5}
 {bone1Name=loop_L1, bone2Name=loop_L, swing1LimitDegrees=10, swing2LimitDegrees=15, twistLimitDegrees=5}
 {bone1Name=loop_L2, bone2Name=loop_L1, swing1LimitDegrees=10, swing2LimitDegrees=10, twistLimitDegrees=5}
 {bone1Name=loop_R, bone2Name=bow_offset_anim, swing1LimitDegrees=10, swing2LimitDegrees=10, twistLimitDegrees=5}
 {bone1Name=loop_R1, bone2Name=loop_R, swing1LimitDegrees=10, swing2LimitDegrees=15, twistLimitDegrees=5}
 {bone1Name=loop_R2, bone2Name=loop_R1, swing1LimitDegrees=10, swing2LimitDegrees=10, twistLimitDegrees=5}
 {bone1Name=tail_R1, bone2Name=tail_R, swing1LimitDegrees=20, swing2LimitDegrees=20, twistLimitDegrees=8}
 {bone1Name=tail_L, bone2Name=bow_offset_anim, swing1LimitDegrees=20, swing2LimitDegrees=20, twistLimitDegrees=8}
 {bone1Name=tail_L1, bone2Name=tail_L, swing1LimitDegrees=20, swing2LimitDegrees=20, twistLimitDegrees=8}
 {bone1Name=bow_offset_anim, bone2Name=bow_mstr, swing1LimitDegrees=45, swing2LimitDegrees=45, twistLimitDegrees=45}
 {bone1Name=bow_mstr, bone2Name=root, swing1LimitDegrees=45, swing2LimitDegrees=45, twistLimitDegrees=45}
 {bone1Name=tail_R, bone2Name=bow_offset_anim, swing1LimitDegrees=20, swing2LimitDegrees=20, twistLimitDegrees=8}
]
