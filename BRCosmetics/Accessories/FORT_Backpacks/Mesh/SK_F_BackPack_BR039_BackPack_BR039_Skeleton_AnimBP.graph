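# Animation Blueprint for the BR039 backpack skeleton that adds physics-driven secondary motion to the accessory.
# The live pose chain starts from the reference pose, offsets the root bone, then runs RigidBody with an overriding Fuzzy Bear Teddy physics asset.
# AnimDynamics nodes for loop and tail bones, plus a speed-scaled float calculation, never reach the output.
# The EventGraph clears DynamicsOK above a pawn speed threshold and caches the owner's mesh in MeshToCopy.
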AnimBlueprint /BRCosmetics/Accessories/FORT_Backpacks/Mesh/SK_F_BackPack_BR039_BackPack_BR039_Skeleton_AnimBP
target_skeleton=/Game/Accessories/FORT_Backpacks/Mesh/SK_F_BackPack_BR039_BackPack_BR039_Skeleton
graphs [
 AnimGraph {nodes [
   n1 = Anim Dynamics - Bone: loop_L(Alpha=1, ExternalForce="0,0,0") {
    bGravityOverrideInSimSpace=true
    lODThreshold=1
    simulationSpace=World
    alphaBoolBlend {blendOption=Linear}
    boundBone {boneName=loop_L}
    physicsBodyDefinitions [{
      sphereCollisionRadius=0
      boundBone {boneName=loop_L}
      boxExtents=<20 20 20>
      constraintSetup {angularLimitsMax=<68.3999 57.5999 46.8>, angularLimitsMin=<-53.9999 -61.1999 -46.8>, angularTarget=<0 0 0>}
      localJointOffset=<0 0 -5>
     }]
   }
   n2 = Anim Dynamics - Bone: loop_R(ComponentPose=@n1, Alpha=1, ExternalForce="0,0,0") {
    bGravityOverrideInSimSpace=true
    lODThreshold=1
    simulationSpace=World
    alphaBoolBlend {blendOption=Linear}
    boundBone {boneName=loop_R}
    physicsBodyDefinitions [{
      sphereCollisionRadius=0
      boundBone {boneName=loop_R}
      boxExtents=<20 20 20>
      constraintSetup {angularLimitsMax=<68.3999 57.5999 46.8>, angularLimitsMin=<-53.9999 -61.1999 -46.8>, angularTarget=<0 0 0>}
      localJointOffset=<0 0 -5>
     }]
   }
   n3 = Anim Dynamics - Bone: tail_R(ComponentPose=@n2, Alpha=1, ExternalForce="0,0,0") {
    bGravityOverrideInSimSpace=true
    lODThreshold=1
    simulationSpace=World
    alphaBoolBlend {blendOption=Linear}
    boundBone {boneName=tail_R}
    physicsBodyDefinitions [{
      sphereCollisionRadius=0
      boundBone {boneName=tail_R}
      boxExtents=<10 10 10>
      constraintSetup {angularLimitsMax=<68.3999 18 46.8>, angularLimitsMin=<-53.9999 -18 -46.8>, angularTarget=<0 0 0>}
      localJointOffset=<2 0 0>
     }]
   }
   Anim Dynamics - Bone: tail_L(ComponentPose=@n3, Alpha=1, ExternalForce="0,0,0") {
    bGravityOverrideInSimSpace=true
    lODThreshold=1
    simulationSpace=World
    alphaBoolBlend {blendOption=Linear}
    boundBone {boneName=tail_L}
    physicsBodyDefinitions [{
      sphereCollisionRadius=0
      boundBone {boneName=tail_L}
      boxExtents=<10 10 10>
      constraintSetup {angularLimitsMax=<68.3999 18 46.8>, angularLimitsMin=<-53.9999 -18 -46.8>, angularTarget=<0 0 0>}
      localJointOffset=<2 0 0>
     }]
   }
   n4 = "Transform (Modify) Bone - Bone: root"(ComponentPose=Local To Component(LocalPose=Local Space Ref Pose()), Translation="0,-1.3,0", Rotation="0,0,5") {
    rotationMode=BMM_Additive
    translationMode=BMM_Replace
    alphaBoolBlend {blendOption=Linear}
    boneToModify {boneName=root}
    rotation {roll=5}
    translation {y=-1.3}
   }
   n5 = RigidBody(ComponentPose=@n4) {
    bClampLinearTranslationLimitToRefPose=true
    bOverrideWorldGravity=true
    lODThreshold=1
    overridePhysicsAsset=/Game/Accessories/FORT_Backpacks/F_MED_Fuzzy_Bear_TEDDY_backpack/Mesh/Fuzzy_Bear_TEDDY_PhysicsAsset
    alphaBoolBlend {blendOption=Linear}
    baseBoneRef {boneName=root}
    componentLinearAccScale=<0.5 0.5 0.5>
    componentLinearVelScale=<0.3 0.3 0.3>
    overrideWorldGravity {z=-980}
   }
   Output Pose(Result=Component To Local(ComponentPose=@n5)) {layerGroup=None}
   AnimGraphNode_SubInput_0 = Input Pose()
   float * float!(
    A=MapRangeClamped(Value=Get PawnSpeed2D(), InRangeA=0, InRangeB=420, OutRangeA=0.6)
    B="To Float (Boolean)"(InBool=Get DynamicsOK() -> (DynamicsOK:bool))
    C:double=0.5
   ) -> (ReturnValue)
  ]}
 EventGraph {nodes [
   Event BlueprintUpdateAnimation()
   > n1 = Branch(Condition="float > float"(A=Get PawnSpeed2D(), B=600) {is=title K2Node_CallFunction})
   > Set DynamicsOK(DynamicsOK:bool=false) -> (Output_Get:bool)
   Set DynamicsOK(execute=@n1.else, DynamicsOK=true)
   Event BlueprintInitializeAnimation()
   > Set MeshToCopy(MeshToCopy:Skeletal Mesh Component Object Reference=GetOwnerSkeletalMeshComponent()) -> (Output_Get:Skeletal Mesh Component Object Reference)
  ]}
]
variables [
 DynamicsOK {type=boolean, default=true}
 MeshToCopy {type=SkeletalMeshComponent}
]
