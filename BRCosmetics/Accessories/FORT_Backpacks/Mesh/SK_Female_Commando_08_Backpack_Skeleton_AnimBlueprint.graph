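# Animation Blueprint for the Female Commando 08 backpack skeleton that adds physical secondary motion to two dynamic bag bones.
# The input pose passes in component space through chained Anim Dynamics nodes for dyn_bag_A then dyn_bag_B, then returns to local space for output.
# Each node simulates a damped, spring-driven body with tight linear and angular limits and low gravity, blended at partial alpha.
# The event graph has no logic.
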AnimBlueprint /BRCosmetics/Accessories/FORT_Backpacks/Mesh/SK_Female_Commando_08_Backpack_Skeleton_AnimBlueprint
target_skeleton=/Game/Accessories/FORT_Backpacks/Mesh/SK_Female_Commando_08_Backpack_Skeleton
graphs [
 AnimGraph {nodes [
   AnimGraphNode_SubInput_0 = Input Pose()
   n1 = Anim Dynamics - Bone: dyn_bag_A(ComponentPose=Local To Component(LocalPose=@AnimGraphNode_SubInput_0), Alpha=0.35, ExternalForce="0,0,0") {
    angularDampingOverride=0.995
    angularSpringConstant=100
    bAngularSpring=true
    bGravityOverrideInSimSpace=true
    bLinearSpring=true
    bOverrideAngularDamping=true
    bOverrideLinearDamping=true
    gravityScale=0.1
    lODThreshold=1
    linearDampingOverride=0.995
    linearSpringConstant=200
    alphaBoolBlend {blendOption=Linear}
    boundBone {boneName=dyn_bag_A}
    componentLinearAccScale=<1 1 1>
    componentLinearVelScale=<1 1 1>
    physicsBodyDefinitions [{
      sphereCollisionRadius=0
      boundBone {boneName=dyn_bag_A}
      boxExtents=<2 2 2>
      constraintSetup {angularLimitsMax=<0 10 0>, angularLimitsMin=<0 -2 0>, angularTarget=<0 0 0>, linearAxesMax=<0 0 2>, linearAxesMin=<0 0 -2>}
      localJointOffset=<0 0 -1>
     }]
   }
   n2 = Anim Dynamics - Bone: dyn_bag_B(ComponentPose=@n1, Alpha=0.35, ExternalForce="0,0,0") {
    angularDampingOverride=0.995
    angularSpringConstant=100
    bAngularSpring=true
    bGravityOverrideInSimSpace=true
    bLinearSpring=true
    bOverrideAngularDamping=true
    bOverrideLinearDamping=true
    gravityScale=0.1
    lODThreshold=1
    linearDampingOverride=0.995
    linearSpringConstant=200
    alphaBoolBlend {blendOption=Linear}
    boundBone {boneName=dyn_bag_B}
    componentLinearAccScale=<1 1 1>
    componentLinearVelScale=<1 1 1>
    physicsBodyDefinitions [{
      sphereCollisionRadius=0
      boundBone {boneName=dyn_bag_B}
      boxExtents=<2.5 2 2>
      constraintSetup {angularLimitsMax=<0 2 0>, angularLimitsMin=<0 -10 0>, angularTarget=<0 0 0>, linearAxesMax=<0 0 2>, linearAxesMin=<0 0 -2>}
      localJointOffset=<0 0 -1>
     }]
   }
   Output Pose(Result=Component To Local(ComponentPose=@n2)) {layerGroup=None}
  ]}
 EventGraph {nodes [Event BlueprintUpdateAnimation()]}
]
variables []
